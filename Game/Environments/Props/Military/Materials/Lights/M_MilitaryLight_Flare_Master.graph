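# Master unlit additive material for military light flares, drawing a soft, tinted sphere-mask glow with depth testing disabled.
# Vertex-offset logic aligns the mesh to the camera with a bias, or collapses it to its pivot when a scene-depth test at that pivot indicates occlusion.
# A vertex-interpolated fade combines that visibility with camera-distance falloff; mobile skips the occlusion test, using pixel-depth fade instead.
# A screen-edge mask is also computed.
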Material /Game/Environments/Props/Military/Materials/Lights/M_MilitaryLight_Flare_Master
graph {
 properties {
  bDisableDepthTest=true
  bEnableResponsiveAA=true
  blendMode=BLEND_Additive
  pixelDepthOffsetMode=PDOM_Legacy
  shadingModel=MSM_Unlit
  usedShadingModels=Unlit
 }
 nodes [
  n1 = Subtract(A=ObjectPositionWS(), B=WorldPosition().XYZ)
  n2 = ObjectPositionWS()
  n3 = WorldPosition()
  n4 = CameraPositionWS()
  n5 = ObjectPositionWS()
  n6 = Multiply(A=Normalize(VectorInput=Subtract(A=@n4, B=@n5)), B=VectorParameter() {parameterName=CameraBias}.RGB)
  n7 = Add(A=Add(A=WorldPosition().XYZ, B=Subtract(A=@n2, B=@n3.XYZ)), B=@n6)
  n8 = Subtract(
   A=SceneDepth(UVs=MaterialFunctionCall(World Position=@n7) {materialFunction=&TransformToClipSpace}.Screen aligned 0-1 UV).R
   B=Subtract(
    A=PixelDepth().R
    B=MaterialFunctionCall(Vector 3=@n6) {materialFunction=/Engine/Functions/Engine_MaterialFunctions02/Utility/VectorLength}.V3 Length
   )
  )
  n9 = ScalarParameter() {defaultValue=-600, parameterName=Offset}
  n10 = Add(
   A=@n6
   B=MaterialFunctionCall!(
    WorldPosition=WorldPosition().XYZ
    PivotLocation
    WS Normals
    Custom Object Basis 1=Constant3Vector() {constant=<rgba 0 1 0 1>}
    Custom Object Basis 2=Constant3Vector() {constant=<rgba 0 0 -1 0>}
    Custom Object Basis 3
   ) {materialFunction=/Engine/Functions/Engine_MaterialFunctions02/WorldPositionOffset/AlignMeshToTheCamera}.World Position Offset
  )
  LinearInterpolate(A=@n1, B=ShadingPathSwitch(Default=If(A=@n8, B=@n9, "A > B"=@n10, "A == B"=@n1, "A < B"=@n1), Mobile=@n10))
  n11 = TextureCoordinate()
  n12 = Constant() {r=0.5}
  n13 = ScalarParameter() {defaultValue=0.1, parameterName=FlareHardness}
  Multiply(
   A=Power(Base=SphereMask(A=@n11, B=@n12, Radius=@n12, Hardness=@n13), Exp=ScalarParameter() {defaultValue=4, parameterName=FlareExp}) {constExponent=3}
   B=VectorParameter() {parameterName=FlareColor, defaultValue=<rgba 0.0681 0.151 0.4 0>}.RGB
  )
  n14 = MaterialFunctionCall!(
   min=ScalarParameter() {defaultValue=1800, parameterName=Start}
   max=ScalarParameter() {defaultValue=1400, parameterName=End}
   value=Distance(A=ObjectPositionWS(), B=CameraPositionWS())
  ) {materialFunction=/Game/Effects/Fort_Effects/Materials/_Functions/InverseLerp_MF}
  Saturate(None=Ceil(None=@n14.Clamped))
  n15 = ConstantBiasScale(None=ScreenPosition().ViewportUV) {bias=-0.5, scale=2}
  n16 = Constant()
  SphereMask(A=Power(Base=Abs(None=@n15)) {constExponent=6}, B=@n16, Radius=Constant() {r=0.95}, Hardness=@n16)
  n17 = Divide(
   A=Subtract(A=PixelDepth().R, B=ScalarParameter() {defaultValue=300, parameterName=gfdsg})
   B=ScalarParameter() {defaultValue=50, parameterName=tmyukjrjk}
  )
  n18 = ShadingPathSwitch(Default=@n8, Mobile=OneMinus(None=Saturate(None=@n17)))
  VertexInterpolator(VS=Multiply(A=Saturate(None=@n18), B=@n14.Clamped))
 ]
}
