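# Additive, unlit, two-sided material for military light cone beams, usable on Niagara particles, with optional particle color tint.
# Vertices flare along normals by UV gradient, plus an optional DanceClub offset function.
# Opacity combines a UV gradient, Fresnel edge softening (cheaper view-angle fade at lower quality) and panning, actor-position-offset fog noise.
# Time-of-day scaling, synced distance fade, camera depth fade and quality-switched depth fade control visibility.
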Material /Game/Environments/Props/Military/Materials/Lights/M_MilitaryLight_LightCone
graph {
 properties {
  bUsedWithMeshParticles=true
  bUsedWithNanite=true
  bUsedWithNiagaraMeshParticles=true
  bUsedWithNiagaraRibbons=true
  bUsedWithNiagaraSprites=true
  blendMode=BLEND_Additive
  pixelDepthOffsetMode=PDOM_Legacy
  shadingModel=MSM_Unlit
  twoSided=true
  usedShadingModels=Unlit
 }
 nodes [
  n1 = ComponentMask(None=TextureCoordinate()) {g=true}
  n2 = VertexNormalWS()
  Add(
   A=Multiply(A=Multiply(A=@n1, B=@n2), B=ScalarParameter() {defaultValue=150, parameterName=FlareAmount})
   B=StaticSwitchParameter(True=MaterialFunctionCall() {materialFunction=/Game/Environments/Sets/DigSite/Materials/MF_Sneezy_DanceClub}.Result, False=Constant()) {parameterName=DanceClub}
  )
  n3 = ParticleColor()
  n4 = MaterialFunctionCall(
   Position=WorldPosition().XYZ
   Fade_Start_Distance=ScalarParameter() {defaultValue=250, parameterName="Fade_Start_Distance (S)"}
   Fade_End_Distance=ScalarParameter() {defaultValue=100, parameterName="Fade_End_Distance (S)"}
  ) {materialFunction=&MF_Distance_SyncedFade_FX}
  n5 = Multiply(A=VectorParameter() {parameterName=Color, defaultValue=<rgba 1.1499 1.8452 3 1>}.RGB, B=@n4.Result)
  MaterialFunctionCall(Emissive=StaticSwitchParameter(True=Multiply(A=@n3.RGB, B=@n5), False=@n5) {parameterName=Use Particle Color?}) {materialFunction=&MF_FX_EmissiveAdjustments}
  n6 = ScalarParameter() {defaultValue=3, parameterName="Value_Morning (S)"}
  n7 = ScalarParameter() {defaultValue=1.25, parameterName="Value_Day (S)"}
  n8 = ScalarParameter() {defaultValue=1, parameterName="Value_Evening (S)"}
  n9 = ScalarParameter() {defaultValue=0.7, parameterName="Value_Night (S)"}
  n10 = ComponentMask(None=TextureCoordinate()) {g=true}
  n11 = OneMinus(None=@n10)
  n12 = ScalarParameter() {defaultValue=1, parameterName=FlipY}
  n13 = ScalarParameter() {defaultValue=0.5, parameterName=ExponentIn}
  n14 = Power(Base=OneMinus(None=Fresnel(ExponentIn=@n13) {baseReflectFraction=0}), Exp=ScalarParameter() {defaultValue=2, parameterName=Exp})
  n15 = MaterialFunctionCall("FadeSensitivity (optional)"=ScalarParameter() {defaultValue=2, parameterName=FadeSensitivity}) {materialFunction=&FadeBasedOnViewAngle}
  n16 = Multiply(
   A=Power(Base=LinearInterpolate(A=@n10, B=@n11, Alpha=@n12), Exp=ScalarParameter() {defaultValue=1.1, parameterName=ConePow})
   B=QualitySwitch(Default=@n14, Low=@n15.Result, High=@n14, Medium=@n15.Result)
  )
  n17 = TextureCoordinate() {vTiling=0.25}
  n18 = Multiply(A=ComponentMask(None=ActorPositionWS()) {g=true, r=true}, B=ScalarParameter() {defaultValue=0.0001, parameterName=ActorPostion_Offset}) {constB=0.005}
  n19 = ScalarParameter() {parameterName=Speed X}
  n20 = ScalarParameter() {defaultValue=0.05, parameterName=Speed Y}
  n21 = Panner(
   Coordinate=Multiply(A=Add(A=@n17, B=@n18), B=ScalarParameter() {defaultValue=1, parameterName=TileFogNoise})
   Speed=Multiply(A=ScalarParameter() {defaultValue=1, parameterName=Speed Multiplier}, B=AppendVector(A=@n19, B=@n20))
  ) {speedY=0.05}
  n22 = Panner(Coordinate=TextureCoordinate()) {speedX=-0.125}
  n23 = LinearInterpolate(
   A=@n16
   B=Multiply(A=@n16, B=TextureSample(UVs=@n21) {texture=&T_Smooth_tile}.G)
   Alpha=OneMinus(None=TextureSample(UVs=@n22) {texture=/Game/Effects/Fort_Effects/Textures/T_HitImpact}.G)
  )
  n24 = QualitySwitch(Default=Multiply(A=DepthFade() {fadeDistanceDefault=25}, B=@n23), Low=@n23)
  n25 = ComponentMask(None=MaterialFunctionCall(
    Fade Length=ScalarParameter() {defaultValue=8192, parameterName=FarFade}
    Fade Offset=ScalarParameter() {defaultValue=1024, parameterName=FarFade_Offset}
   ) {materialFunction=&CameraDepthFade}.Result) {r=true}
  n26 = Multiply(
   A=Multiply(
    A=MaterialFunctionCall(Value_Morning=@n6, Value_Day=@n7, Value_Evening=@n8, Value_Night=@n9) {materialFunction=&Scalar_Time_of_Day}.Result
    B=Multiply(A=@n4.Result, B=@n24)
   )
   B=OneMinus(None=Clamp(None=@n25))
  )
  StaticSwitchParameter(True=Multiply(A=@n3.A, B=@n26), False=@n26) {parameterName=Use Particle Color?}
 ]
}
